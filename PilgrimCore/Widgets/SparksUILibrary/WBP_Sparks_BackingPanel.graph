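# Sparks UI backing panel widget that layers a Darken image over a material-driven Fill image.
# Design-time PreConstruct and runtime Construct toggle Darken from DarkenLayer, then apply the panel type and corner radius enums.
# Panel type selects Fill's material; corner radius sets the Corner_Radius parameter on both images' dynamic materials.
# PlayHover and PlayUnhover run the Hover animation forward or reverse, easing Darken's opacity between full and half.
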WidgetBlueprint /PilgrimCore/Widgets/SparksUILibrary/WBP_Sparks_BackingPanel
animations [Hover {
  binding_count=1
  length_seconds=0.2
  object=&self:Hover
  sequence {bindings [Darken {
     id=#1
     possessed_class=Image
     tracks [{
       type=MovieSceneFloatTrack
       display_name=Render Opacity
       sections [{type=MovieSceneFloatSection, channels [None {type=MovieSceneScriptingFloatChannel, keys [[0,1],[4,0.5]]}]}]
      }]
    }]}
 }]
graphs [
 EventGraph {nodes [
   n1 = Event PreConstruct()
   > n2 = Branch(Condition=@n1.IsDesignTime)
   SetVisibility(
    execute=[@n3.then, @n2.then]
    self=Get Darken() -> (Darken:Image Object Reference)
    InVisibility=Select(Option 0=Collapsed, Option 1=SelfHitTestInvisible, Index=Get DarkenLayer() -> (DarkenLayer:bool))
   )
   > SetBackingType!(
    execute
    self
    Index:SparksBackplateType_Enumeration Enum=Get ESparksBackingPanelType() -> (ESparksBackingPanelType:SparksBackplateType_Enumeration Enum=NewEnumerator0)
   ) -> (then)
   > SetCornerRadius!(execute, self, Index:SparksBackplateCornerRadius Enum=Get CornerRadius() -> (CornerRadius:SparksBackplateCornerRadius Enum=NewEnumerator0)) -> (then)
   n3 = Event Construct()
   Event Tick()
  ]}
 PlayHover {nodes [
   PlayHover() {is=title K2Node_FunctionEntry}
   > PlayAnimationForward(InAnimation=Get Hover() -> (Hover:Widget Animation Object Reference))
  ]}
 PlayUnhover {nodes [
   PlayUnhover() {is=title K2Node_FunctionEntry}
   > PlayAnimationReverse(InAnimation=Get Hover())
  ]}
 SetBackingType {nodes [
   n1 = SetBackingType!() -> (then, Index:SparksBackplateType_Enumeration Enum=NewEnumerator0) {is=title K2Node_FunctionEntry}
   > SetBrushFromMaterial(
    self=Get Fill() -> (Fill:Image Object Reference)
    Material=Select!(
     NewEnumerator0:Material Interface Object Reference=/SparksUICommon/Materials/MI_UI_Pilgrim_Panel_Underlayer_Complex
     NewEnumerator1:Material Interface Object Reference=/PilgrimCore/Widgets/Materials/MI_UI_Pilgrim_Panel_Underlayer_ColorFill
     NewEnumerator2:Material Interface Object Reference=/PilgrimCore/Widgets/Materials/MI_UI_Pilgrim_Panel_Underlayer_VGradient
     NewEnumerator3:Material Interface Object Reference=/PilgrimCore/Widgets/Materials/MI_UI_Pilgrim_Panel_Underlayer_HGradient
     Index:SparksBackplateType_Enumeration Enum=@n1
    ) -> (ReturnValue:Material Interface Object Reference)
   )
  ]}
 SetCornerRadius {nodes [
   n1 = SetCornerRadius!() -> (then, Index:SparksBackplateCornerRadius Enum=NewEnumerator0) {is=title K2Node_FunctionEntry}
   > n2 = GetDynamicMaterial(self=Get Fill())
   > n3 = GetDynamicMaterial(self=Get Darken())
   > SetScalarParameterValue(
    self=[@n3, @n2]
    ParameterName=Corner_Radius
    Value=Select!(
     NewEnumerator0:double=0.125
     NewEnumerator1:double=0.25
     NewEnumerator2:double=0.333
     NewEnumerator3:double=0.5
     NewEnumerator4:double=0.666
     Index:SparksBackplateCornerRadius Enum=@n1
    ) -> (ReturnValue:double)
   )
  ]}
]
variables [
 BlurFrame {type=boolean, default=true}
 CornerRadius {type=SparksBackplateCornerRadius, default=Medium}
 DarkenLayer {type=boolean, default=true}
 ESparksBackingPanelType {type=SparksBackplateType_Enumeration, default=Complex Gradient}
]
widgets {
 info {inheritedWidgetCount=0, parentClass=&UserWidget, rootWidgetClass=&Overlay, widgetCount=3}
 widgets [
  All {widgetClassPath=&Overlay, widget=All}
  Fill {
   widgetClassPath=&Image
   bIsVariable=true
   parent=All
   widget=Fill
   properties {brush {
     bIsDynamicallyLoaded=false
     drawAs=Image
     imageType=NoImage
     mirroring=NoMirror
     resourceName=None
     resourceObject=&MI_UI_Pilgrim_Panel_Underlayer_Complex
     tiling=NoTile
     imageSize=<32 32>
     margin=<ltrb 0 0 0 0>
     outlineSettings {
      bUseBrushTransparency=false
      roundingType=HalfHeightRadius
      width=0
      color {colorUseRule=UseColor_Specified, specifiedColor=<rgba 0 0 0 0>}
      cornerRadii=<0 0 0 0>
     }
     tintColor {colorUseRule=UseColor_Specified, specifiedColor=<rgba 1 1 1 1>}
     uVRegion {bIsValid=false, max=<0 0>, min=<0 0>}
    }}
   slot {class=OverlaySlot, properties {horizontalAlignment=HAlign_Fill, verticalAlignment=VAlign_Fill}}
  }
  Darken {
   widgetClassPath=&Image
   bIsVariable=true
   parent=All
   widget=Darken
   properties {
    visibility=SelfHitTestInvisible
    brush {
     bIsDynamicallyLoaded=false
     drawAs=Image
     imageType=NoImage
     mirroring=NoMirror
     resourceName=None
     resourceObject=/SparksUICommon/Materials/MI_UI_Pilgrim_Panel_Underlayer_Darken
     tiling=NoTile
     imageSize=<32 32>
     margin=<ltrb 0 0 0 0>
     outlineSettings {
      bUseBrushTransparency=false
      roundingType=HalfHeightRadius
      width=0
      color {colorUseRule=UseColor_Specified, specifiedColor=<rgba 0 0 0 0>}
      cornerRadii=<0 0 0 0>
     }
     tintColor {colorUseRule=UseColor_Specified, specifiedColor=<rgba 1 1 1 1>}
     uVRegion {bIsValid=false, max=<0 0>, min=<0 0>}
    }
   }
   slot {class=OverlaySlot, properties {horizontalAlignment=HAlign_Fill, verticalAlignment=VAlign_Fill}}
  }
 ]
}
