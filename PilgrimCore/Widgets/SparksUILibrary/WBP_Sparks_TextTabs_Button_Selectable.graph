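# WBP_Sparks_TextTabs_Button_Selectable is a PilgrimCore Sparks UI text-tab button derived from the UIKit selectable button.
# It adds no widgets, and its pre-construct, construct and hover events only call the parent.
# Focus and selection both trigger the parent's On Selected, while unfocus and deselection both trigger On Deselected.
# An unconnected tab-backing brush node, an empty Tick and two unreferenced booleans are also present.
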WidgetBlueprint /PilgrimCore/Widgets/SparksUILibrary/WBP_Sparks_TextTabs_Button_Selectable
graphs [EventGraph {nodes [
   n1 = Event PreConstruct()
   > Parent: PreConstruct(IsDesignTime=@n1.IsDesignTime)
   Event Construct()
   > Parent: Construct()
   Event On Hovered()
   > Parent: On Hovered() {is=title K2Node_CallParentFunction}
   Event On Unhovered()
   > Parent: On Unhovered() {is=title K2Node_CallParentFunction}
   n2 = Event On Focused() {is=title K2Node_Event}
   Parent: On Selected(execute=[@n2.then, @n3.then]) {is=title K2Node_CallParentFunction}
   n3 = Event On Selected()
   n4 = Event On Deselected()
   Parent: On Deselected(execute=[@n4.then, @n5.then]) {is=title K2Node_CallParentFunction}
   n5 = Event On Unfocused() {is=title K2Node_Event}
   SetBrushFromMaterial(Material=/PilgrimCore/Widgets/Materials/MI_Pilgrim_TabBacking)
   Event Tick()
  ]}]
variables [
 InitialMouseLeave {type=boolean}
 IsSelected {type=boolean}
]
widgets {
 info {
  inheritedWidgetCount=0
  parentClass=/FortUILibrary/UIKit/Buttons/WBP_UIKit_Button_Selectable.WBP_UIKit_Button_Selectable_C
  rootWidgetClass=null
  widgetCount=0
 }
 widgets []
}
